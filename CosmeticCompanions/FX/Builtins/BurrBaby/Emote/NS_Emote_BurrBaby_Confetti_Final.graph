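# Niagara confetti burst for the BurrBaby companion's emote, spraying upward from two points either side of the origin.
# NE_Confetti sprites launch in cones, then drift under drag, gravity and curl noise, shrinking late in life.
# Short-lived flare sprites and two mesh spark emitters share the paired offsets via partition switches, with flare alpha scaled by time of day.
# Four LWE_ emitters (Glow, Serpentine, Sparkles, Sparks) list no modules or renderers here.
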NiagaraSystem /CosmeticCompanions/FX/Builtins/BurrBaby/Emote/NS_Emote_BurrBaby_Confetti_Final
dependencies {
 usedDataInterfaces [&NiagaraDataInterfaceCurve]
 usedDynamicInputs [&MakeCustomLinearColorFromBool]
 usedModules [
  &self:SystemSpawnScript.SystemScriptSource.SystemScriptGraph.NiagaraNodeAssignment_1.SetVariables_#1
  &SystemState
 ]
}
emitters {
 LWE_Glow {rendererClasses [], renderers []}
 LWE_Serpentine {rendererClasses [], renderers []}
 LWE_Sparkles {rendererClasses [], renderers []}
 LWE_Sparks {rendererClasses [], renderers []}
 NE_Confetti {
  emitterUpdateScript [
   &EmitterState(Life Cycle Mode=::NewEnumerator1, Loop Behavior=::NewEnumerator1, Loop Duration=0.5, UseLoopDelay=false)
   &SpawnRate(SpawnRate=50, Use Spawn Probability=false)
  ]
  particleSpawnScript [
   &PartitionParticles()
   &InitializeParticle(
    Color=<rgba 1 0 0 1>
    Lifetime Max=3
    Lifetime Min=1.5
    Lifetime Mode=::NewEnumerator1
    Position Mode=::NewEnumerator0
    Sprite Rotation Angle=&RandomRangeFloat(Minimum=0, Maximum=1)
    Sprite Rotation Mode=::NewEnumerator1
    Sprite Size Mode=::NewEnumerator1
    Write Material Random=true
   ) {module=InitializeParticle001}
   &SystemLocation(
    MaskSpawnGroup=false
    Offset=&MergeHelper_VectorSwitch(0=<-30 0 0>, 1=<30 0 0>, 2=?, 3=?, Index=$Particles.PartitionParticles.Partition, Num=::NewEnumerator0)
   )
   /Niagara/Modules/Spawn/Location/ConeLocation(Cone Angle=10, Cone Axis:Vector3f=<0 0 1>, Cone Length=50, MaskSpawnGroup=false, Module.Override Randomness=false)
   &AddVelocityInCone(
    Cone Angle=25
    Cone Axis=<0 0 1>
    Module.Use Velocity Falloff On Cone Axis=false
    Velocity Strength=&RandomRangeFloat(Minimum=200, Maximum=500)
   )
  ]
  particleUpdateScript [
   &ParticleState()
   &UpdateMeshOrientation(Rotation Vector=<1 1 1>)
   &Drag(Drag=&RandomRangeFloat(Minimum=1, Maximum=2.5), Use Linear Drag=true, Use Rotational Drag=false)
   &GravityForce(Gravity=&RandomRangeVector(Minimum=<0 0 -50>, Maximum=<1 1 -150>))
   &CurlNoiseForce(
    Noise Frequency=16
    Noise Strength=&FloatFromCurve(
     FloatCurve=DataInterface() {propertyValues {Curve {keys [{interpMode=RCIM_Cubic, tangentMode=RCTM_User, time=0, value=0}, {interpMode=RCIM_Cubic, tangentMode=RCTM_User, time=0.5, value=1}]}}}
     Scale Curve=300
    )
    Pan Noise=true
    Pan Noise Field=<0.05 0 0.2>
   )
   &SolveForcesAndVelocity(Clamp Velocity=false, Limit Acceleration=false)
   &ScaleColor(
    Scale RGB=&MakeVector(
     X=1
     Y=&FloatFromCurve(FloatCurve=DataInterface() {propertyValues {Curve {keys [<at 0 1>,<at 0.9976 1>]}}})
     Z=1
    )
    ScaleA=true
    ScaleRGB=true
   )
   &ScaleSpriteSize(Scale Factor=&Vector2DFromFloat(Value=&FloatFromCurve(FloatCurve=DataInterface() {propertyValues {Curve {keys [
          {arriveTangent=-0.4596, interpMode=RCIM_Cubic, leaveTangent=-0.4596, tangentMode=RCTM_User, time=0.55, value=1}
          {arriveTangent=-4.9031, interpMode=RCIM_Cubic, leaveTangent=-4.9031, tangentMode=RCTM_User, time=1, value=0}
         ]}}})))
   &DynamicMaterialParameters(
    Index 0 Param 1=10
    Index 1 Param 1=1
    Module.Param0WriteEnabled=true
    Module.Param0WriteEnabled1=true
    Module.Param1WriteEnabled=true
    Module.Param1WriteEnabled1=true
    Module.Param2WriteEnabled=true
    Module.Param2WriteEnabled1=true
    Module.Param3WriteEnabled=true
    Module.Param3WriteEnabled1=true
    Write Parameter Index 1=true
   )
  ]
  rendererClasses [&NiagaraSpriteRendererProperties]
  renderers [{rendererClass=&NiagaraSpriteRendererProperties}]
 }
 NE_Flare {
  emitterUpdateScript [
   &EmitterState(Inactive Response=::NewEnumerator1, Life Cycle Mode=::NewEnumerator1, Loop Behavior=::NewEnumerator1, Loop Duration=0.35, UseLoopDelay=false)
   &SpawnRate(SpawnRate=15, Use Spawn Probability=false)
  ]
  particleSpawnScript [
   &PartitionParticles()
   &InitializeParticle(
    Color=<rgba 2 0.575 0.1 1>
    Lifetime=0.2
    Position Offset=&MergeHelper_VectorSwitch(0=<-30 0 5>, 1=<30 0 5>, 2=?, 3=?, Index=$Particles.PartitionParticles.Partition, Num=::NewEnumerator0)
    Position Offset Coordinate Space=::Local
    Sprite Size Mode=::NewEnumerator3
    Uniform Sprite Size=65
    UsePositionOffset=true
    Write Material Random=true
   )
   &CameraOffset(Camera Offset Amount=50)
  ]
  particleUpdateScript [
   &ParticleState()
   &ScaleSpriteSize(Uniform Curve Sprite Scale=DataInterface() {propertyValues {Curve {keys [{tangentMode=RCTM_User, time=0, value=1}, {tangentMode=RCTM_User, time=1, value=1.25}]}}})
   &SolveForcesAndVelocity(Clamp Velocity=false, Limit Acceleration=false)
   &ScaleColor(
    Scale Alpha=&FNTimeOfDayScalar(Night Scale=1, Morning Scale=0.5, Day Scale=0.35, Evening Scale=0.5)
    Scale RGB=&VectorFromFloat(Value=&FloatFromCurve(FloatCurve=DataInterface() {propertyValues {Curve {keys [<at 0 0>,<at 0.1 1>,<at 0.8 1>,<at 1 0>]}}}))
    ScaleA=true
    ScaleRGB=true
   )
   &DynamicMaterialParameters(
    Index 0 Param 1=0.75
    Index 0 Param 2=1
    Module.Index0Param0Write=true
    Module.Index0Param1Write=true
    Module.Index0Param2Write=false
    Module.Index0Param3Write=false
   )
  ]
  rendererClasses [&NiagaraSpriteRendererProperties]
  renderers [{rendererClass=&NiagaraSpriteRendererProperties}]
 }
 NE_Sparks_01 {
  emitterUpdateScript [
   &EmitterState(Inactive Response=::NewEnumerator1, Life Cycle Mode=::NewEnumerator1, Loop Behavior=::NewEnumerator1, Loop Duration=0.35, UseLoopDelay=false)
   &SpawnRate(SpawnRate=40, Use Spawn Probability=false)
  ]
  particleSpawnScript [
   &PartitionParticles()
   &InitializeParticle(
    Color=<rgba 25 7.1331 1.0417 1>
    Lifetime=0.25
    Mesh Scale Max=<1 1 3>
    Mesh Scale Min=<1 1 2>
    Mesh Scale Mode=::NewEnumerator2
    Position Offset=&MergeHelper_VectorSwitch(0=<-30 0 0>, 1=<30 0 0>, 2=?, 3=?, Index=$Particles.PartitionParticles.Partition, Num=::NewEnumerator0)
    Position Offset Coordinate Space=::Local
    UsePositionOffset=true
    Write Material Random=true
   )
   &InitialMeshOrientation(
    Mesh Orientation Mode=::NewEnumerator2
    Orientation Coordinate Space=::Local
    Rotation=&MakeVector(X=0, Y=0, Z=&RandomRangeFloat(Minimum=0, Maximum=1))
    Use Rotation Vector=false
   ) {module=InitialMeshOrientation001}
  ]
  particleUpdateScript [
   &ParticleState()
   &ScaleColor(
    Scale Alpha=&FloatFromCurve(FloatCurve=DataInterface() {propertyValues {Curve {keys [
         {arriveTangent=18.7405, interpMode=RCIM_Cubic, leaveTangent=18.7405, tangentMode=RCTM_User, time=0, value=0}
         {interpMode=RCIM_Cubic, tangentMode=RCTM_User, time=0.1, value=1}
         {interpMode=RCIM_Cubic, tangentMode=RCTM_User, time=0.5, value=1}
         {interpMode=RCIM_Cubic, tangentMode=RCTM_User, time=1, value=0}
        ]}}})
    ScaleA=true
    ScaleRGB=true
   )
   &DynamicMaterialParameters(
    Index 0 Param 1=&FloatFromCurve(FloatCurve=DataInterface() {propertyValues {Curve {keys [
         {arriveTangent=11.6708, interpMode=RCIM_Cubic, leaveTangent=11.6708, tangentMode=RCTM_User, time=0, value=2}
         {arriveTangent=0.5512, interpMode=RCIM_Cubic, leaveTangent=0.5512, tangentMode=RCTM_User, time=1, value=15}
        ]}}})
    Index 0 Param 2=&FloatFromCurve(FloatCurve=DataInterface() {propertyValues {Curve {keys [
         {arriveTangent=-0.0433, interpMode=RCIM_Cubic, leaveTangent=-0.0433, tangentMode=RCTM_User, time=0, value=-1}
         {arriveTangent=4.1024, interpMode=RCIM_Cubic, leaveTangent=4.1024, tangentMode=RCTM_User, time=1, value=1}
        ]}}})
    Index 0 Param 3=&RandomRangeFloat(Minimum=2, Maximum=3)
    Module.Index0Param0Write=true
    Module.Index0Param1Write=true
    Module.Index0Param2Write=true
    Module.Index0Param3Write=true
   )
  ]
  rendererClasses [&NiagaraMeshRendererProperties]
  renderers [{rendererClass=&NiagaraMeshRendererProperties, rendererIndex=0}]
 }
 NE_Sparks_02 {
  emitterUpdateScript [
   &EmitterState(Inactive Response=::NewEnumerator1, Life Cycle Mode=::NewEnumerator1, Loop Behavior=::NewEnumerator1, Loop Duration=0.35, UseLoopDelay=false)
   &SpawnRate(SpawnRate=10, Use Spawn Probability=false)
  ]
  particleSpawnScript [
   &PartitionParticles()
   &InitializeParticle(
    Color=<rgba 5 1.4266 0.2083 1>
    Lifetime=0.25
    Mesh Scale Max=<3 3 5>
    Mesh Scale Min=<1 1 2>
    Mesh Scale Mode=::NewEnumerator2
    Position Offset=&MergeHelper_VectorSwitch(0=<-30 0 0>, 1=<30 0 0>, 2=?, 3=?, Index=$Particles.PartitionParticles.Partition, Num=::NewEnumerator0)
    Position Offset Coordinate Space=::Local
    UsePositionOffset=true
    Write Material Random=true
   )
   &InitialMeshOrientation(
    Mesh Orientation Mode=::NewEnumerator2
    Orientation Coordinate Space=::Local
    Rotation=&MakeVector(X=0, Y=0, Z=&RandomRangeFloat(Minimum=0, Maximum=1))
    Use Rotation Vector=true
   ) {module=InitialMeshOrientation001}
  ]
  particleUpdateScript [
   &ParticleState()
   &ScaleColor(
    Scale Alpha=&FloatFromCurve(FloatCurve=DataInterface() {propertyValues {Curve {keys [
         {arriveTangent=18.7405, interpMode=RCIM_Cubic, leaveTangent=18.7405, tangentMode=RCTM_User, time=0, value=0}
         {interpMode=RCIM_Cubic, tangentMode=RCTM_User, time=0.1, value=1}
         {interpMode=RCIM_Cubic, tangentMode=RCTM_User, time=0.5, value=1}
         {interpMode=RCIM_Cubic, tangentMode=RCTM_User, time=1, value=0}
        ]}}})
    ScaleA=true
    ScaleRGB=true
   )
   &DynamicMaterialParameters(
    Index 0 Param 1=&FloatFromCurve(FloatCurve=DataInterface() {propertyValues {Curve {keys [
         {arriveTangent=11.6708, interpMode=RCIM_Cubic, leaveTangent=11.6708, tangentMode=RCTM_User, time=0, value=1}
         {arriveTangent=0.5512, interpMode=RCIM_Cubic, leaveTangent=0.5512, tangentMode=RCTM_User, time=1, value=15}
        ]}}})
    Index 0 Param 2=&FloatFromCurve(FloatCurve=DataInterface() {propertyValues {Curve {keys [
         {arriveTangent=-0.0433, interpMode=RCIM_Cubic, leaveTangent=-0.0433, tangentMode=RCTM_User, time=0, value=-1}
         {arriveTangent=4.1024, interpMode=RCIM_Cubic, leaveTangent=4.1024, tangentMode=RCTM_User, time=1, value=1}
        ]}}})
    Index 0 Param 3=&RandomRangeFloat(Minimum=1, Maximum=6)
    Module.Index0Param0Write=true
    Module.Index0Param1Write=true
    Module.Index0Param2Write=true
    Module.Index0Param3Write=true
   )
  ]
  rendererClasses [&NiagaraMeshRendererProperties]
  renderers [{rendererClass=&NiagaraMeshRendererProperties, rendererIndex=0}]
 }
}
summary {
 systemName=NS_Emote_BurrBaby_Confetti_Final
 emitters [
  {emitterName=NE_Sparks_02, rendererClasses [&NiagaraMeshRendererProperties]}
  {emitterName=NE_Flare, rendererClasses [&NiagaraSpriteRendererProperties]}
  {emitterName=NE_Sparks_01, rendererClasses [&NiagaraMeshRendererProperties]}
  {emitterName=NE_Confetti, rendererClasses [&NiagaraSpriteRendererProperties]}
  {emitterName=LWE_Serpentine, rendererClasses []}
  {emitterName=LWE_Glow, rendererClasses []}
  {emitterName=LWE_Sparks, rendererClasses []}
  {emitterName=LWE_Sparkles, rendererClasses []}
 ]
 userVariables [
  User.bIsFrontEnd {type=NiagaraBool}
  User.bIsFrontEndPreview {type=NiagaraBool}
  User.RemoveAirflow {type=NiagaraBool}
  User.RibbonIDOffset {type=NiagaraInt32}
 ]
}
